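# Control Rig for the SafeThyme male face accessory that offsets five hair root deform bones as the head tilts, likely as a corrective against body clipping.
# Head rotation relative to spine_05 becomes Euler angles, with Z remapped into the blend weight.
# The executed chain applies mostly fixed per-bone rotations, plus X-driven translations on right-side roots.
# An alternative angle-driven chain and a debug-drawing spherical pose reader are disconnected from execution.
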ControlRigBlueprint /BRCosmetics/Characters/Player/Male/Medium/Bodies/M_MED_SafeThyme/Meshes/Parts/M_MED_SafeThyme_FaceAcc_CtrlRig
graphs [
 RigVMFunctionLibrary {links [], nodes []}
 RigVMModel {nodes [
   n1 = &RigVMFunction_MathQuaternionToEuler(
    RotationOrder=ZYX
    Value=&RigUnit_GetRelativeTransformForItem(Child=(Type=Bone,Name="head"), Parent=(Type=Bone,Name="spine_05"), bChildInitial=false, bParentInitial=false).RelativeTransform.Rotation
   )
   Multiply = &RigVMFunction_MathDoubleMul(B=0.1, A=@n1.Result.Z)
   n2 = &RigVMFunction_MathDoubleRemap(SourceMaximum=-90, SourceMinimum=0, TargetMaximum=1, TargetMinimum=0, Value=@n1.Result.Z)
   RigVMFunction_MathQuaternionFromEuler = &RigVMFunction_MathQuaternionFromEuler(Euler=(X=0,Y=-45,Z=0), Euler.Y=&RigVMFunction_MathDoubleMul(B=0.5, A=@n1.Result.Z).Result)
   &RigUnit_ModifyTransforms(
    ExecutePin=()
    ItemToModify=((Item=(Type=Bone,Name="deform_hair_root_l"),Transform=(Rotation=(X=0,Y=0,Z=0,W=1),Translation=(X=-8,Y=0,Z=0),Scale3D=(X=1,Y=1,Z=1))))
    WeightMinimum=0
    ItemToModify.0.Transform.Translation.X=@Multiply.Result
    Weight=@n2.Result
    ItemToModify.0.Transform.Rotation=@RigVMFunction_MathQuaternionFromEuler.Result
   )
   > &RigUnit_ModifyTransforms(
    ItemToModify=((Item=(Type=Bone,Name="deform_hair_root_bk_l"),Transform=(Rotation=(X=0,Y=0,Z=0,W=1),Translation=(X=-8,Y=0,Z=0),Scale3D=(X=1,Y=1,Z=1))))
    WeightMinimum=0
    ItemToModify.0.Transform.Translation.X=@Multiply.Result
    Weight=@n2.Result
    ItemToModify.0.Transform.Rotation=@RigVMFunction_MathQuaternionFromEuler.Result
   )
   > &RigUnit_ModifyTransforms(
    ItemToModify=((Item=(Type=Bone,Name="deform_hair_root_bk_mid"),Transform=(Rotation=(X=0,Y=0,Z=0,W=1),Translation=(X=-7,Y=0,Z=0),Scale3D=(X=1,Y=1,Z=1))))
    WeightMinimum=0
    ItemToModify.0.Transform.Translation.X=@Multiply.Result
    Weight=@n2.Result
    ItemToModify.0.Transform.Rotation=@RigVMFunction_MathQuaternionFromEuler.Result
   )
   > &RigUnit_ModifyTransforms(
    ItemToModify=((Item=(Type=Bone,Name="deform_hair_root_bk_r"),Transform=(Rotation=(X=0,Y=0,Z=0,W=1),Translation=(X=-8,Y=0,Z=0),Scale3D=(X=1,Y=1,Z=1))))
    WeightMinimum=0
    ItemToModify.0.Transform.Translation.X=@Multiply.Result
    Weight=@n2.Result
    ItemToModify.0.Transform.Rotation=@RigVMFunction_MathQuaternionFromEuler.Result
   )
   > &RigUnit_ModifyTransforms(
    ItemToModify=((Item=(Type=Bone,Name="deform_hair_root_r"),Transform=(Rotation=(X=0,Y=0,Z=0,W=1),Translation=(X=-8,Y=0,Z=0),Scale3D=(X=1,Y=1,Z=1))))
    WeightMinimum=0
    ItemToModify.0.Transform.Translation.X=@Multiply.Result
    Weight=@n2.Result
    ItemToModify.0.Transform.Rotation=@RigVMFunction_MathQuaternionFromEuler.Result
   )
   RigUnit_BeginExecution = &RigUnit_BeginExecution()
   &RigUnit_ModifyTransforms(
    ItemToModify=((Item=(Type=Bone,Name="head"),Transform=(Rotation=(X=0,Y=0,Z=0,W=1),Translation=(X=0,Y=0,Z=0),Scale3D=(X=1,Y=1,Z=1))))
    Weight=1
    WeightMinimum=0
    ExecutePin=@RigUnit_BeginExecution.ExecutePin
   )
   > &RigUnit_ModifyTransforms(
    ItemToModify=((Item=(Type=Bone,Name="deform_hair_root_l"),Transform=(Rotation=(X=0,Y=0,Z=-0.3827,W=0.9239),Translation=(X=0,Y=0,Z=0),Scale3D=(X=1,Y=1,Z=1))))
    WeightMinimum=0
    Weight=@n2.Result
   )
   > &RigUnit_ModifyTransforms(
    ItemToModify=((Item=(Type=Bone,Name="deform_hair_root_bk_l"),Transform=(Rotation=(X=-0.0828,Y=-0.2,Z=-0.3736,W=0.902),Translation=(X=0,Y=0,Z=0),Scale3D=(X=1,Y=1,Z=1))))
    WeightMinimum=0
    Weight=@n2.Result
    ItemToModify.0.Transform.Rotation=&RigVMFunction_MathQuaternionFromEuler(Euler=(X=0,Y=0,Z=-45)).Result
   )
   > &RigUnit_ModifyTransforms(
    ItemToModify=((Item=(Type=Bone,Name="deform_hair_root_bk_mid"),Transform=(Rotation=(X=0,Y=-0.2164,Z=0,W=0.9763),Translation=(X=-7,Y=0,Z=0),Scale3D=(X=1,Y=1,Z=1))))
    WeightMinimum=0
    Weight=@n2.Result
   )
   > &RigUnit_ModifyTransforms(
    ItemToModify=((Item=(Type=Bone,Name="deform_hair_root_bk_r"),Transform=(Rotation=(X=0,Y=-0.3827,Z=0,W=0.9239),Translation=(X=-8,Y=15,Z=0),Scale3D=(X=1,Y=1,Z=1))))
    WeightMinimum=0
    ItemToModify.0.Transform.Translation.Y=&RigVMFunction_MathDoubleMul(B=-0.35, A=@n1.Result.X).Result
    ItemToModify.0.Transform.Translation.X=&RigVMFunction_MathDoubleMul(B=0.3, A=@n1.Result.X).Result
    Weight=@n2.Result
   )
   > &RigUnit_ModifyTransforms(
    ItemToModify=((Item=(Type=Bone,Name="deform_hair_root_r"),Transform=(Rotation=(X=0,Y=0,Z=0.3827,W=0.9239),Translation=(X=-8,Y=5,Z=0),Scale3D=(X=1,Y=1,Z=1))))
    WeightMinimum=0
    ItemToModify.0.Transform.Translation.Y=&RigVMFunction_MathDoubleMul(B=-0.2, A=@n1.Result.X).Result
    ItemToModify.0.Transform.Translation.X=&RigVMFunction_MathDoubleMul(B=0.3, A=@n1.Result.X).Result
    Weight=@n2.Result
   )
   &RigUnit_SphericalPoseReader(
    ActiveRegionScaleFactors=(PositiveWidth=1,NegativeWidth=3,PositiveHeight=3,NegativeHeight=1)
    ActiveRegionSize=0.1
    Debug=(bDrawDebug=True,bDraw2D=False,bDrawLocalAxes=False,DebugScale=25,DebugSegments=20,DebugThickness=0.25)
    DriverAxis=(X=1,Y=0,Z=0)
    DriverItem=(Type=Bone,Name="head")
    ExecutePin=()
    FalloffRegionScaleFactors=(PositiveWidth=1,NegativeWidth=2,PositiveHeight=1,NegativeHeight=1)
    FalloffSize=0.3
    FlipHeightScaling=False
    FlipWidthScaling=False
    OptionalParentItem=(Type=Bone,Name="spine_05")
    RotationOffset=(X=-65,Y=0,Z=0)
   )
  ]}
]
hierarchy [
 root {loc=<0 0 0>, rot=<0 0 0 1>}
 pelvis {parent=root, loc=<0 -0.7071 93.1191>, rot=<0.0127 0.707 -0.0127 -0.707>}
 spine_01 {parent=pelvis, loc=<6.7494 -0.6388 0>, rot=<0 0 0.034 -0.9994>}
 spine_02 {parent=spine_01, loc=<7.2518 -0.035 0>, rot=<0 0 0.0238 -0.9997>}
 spine_03 {parent=spine_02, loc=<6.0104 0 0>, rot=<0 0 0.0149 -0.9999>}
 spine_04 {parent=spine_03, loc=<8.1024 0 0>, rot=<0 0 -0.0132 -0.9999>}
 spine_05 {parent=spine_04, loc=<10.0146 0 0>, rot=<0 0 -0.0015 -1>}
 neck_01 {parent=spine_05, loc=<11.3138 0.1529 0>, rot=<0 0 -0.2003 -0.9797>}
 neck_02 {parent=neck_01, loc=<4.5627 0 0>, rot=<0 0 0.0504 -0.9987>}
 head {parent=neck_02, loc=<5.3914 0 0>, rot=<0 0 0.115 -0.9934>}
 dyn_hair_base_01_r {parent=head, loc=<7.324 2.9536 9.4672>, rot=<-0.0288 0.1877 0.9818 -0.0098>}
 dyn_hair_01_r {parent=dyn_hair_base_01_r, loc=<0 0 0>, rot=<0 0 0 1>}
 dyn_hair_01_02_r {parent=dyn_hair_01_r, loc=<8.3974 -0.287 0>, rot=<0 0 0 1>}
 dyn_hair_base_fr_r {parent=head, loc=<14.3574 6.3641 10.607>, rot=<-0.0114 0.1754 0.982 0.0684>}
 dyn_hair_01_fr_r {parent=dyn_hair_base_fr_r, loc=<0 0 0>, rot=<0 0 0 1>}
 dyn_hair_base_fr_mid {parent=head, loc=<14.2319 11.72 -0.0287>, rot=<-0.048 -0.7067 -0.7049 -0.0378>}
 dyn_hair_01_fr_mid {parent=dyn_hair_base_fr_mid, loc=<0 0 0>, rot=<0 0 0 1>}
 dyn_hair_base_fr_l {parent=head, loc=<13.6903 6.0731 -10.1681>, rot=<-0.0665 -0.9489 -0.3065 0.036>}
 dyn_hair_01_fr_l {parent=dyn_hair_base_fr_l, loc=<0 0 0>, rot=<0 0 0 1>}
 dyn_hair_base_01_l {parent=head, loc=<9.8069 2.1747 -10.0859>, rot=<0.0013 -0.994 -0.0996 0.0453>}
 dyn_hair_01_l {parent=dyn_hair_base_01_l, loc=<0 0 0>, rot=<0 0 0 1>}
 dyn_hair_01_02_l {parent=dyn_hair_01_l, loc=<9.7361 0.6323 0.1591>, rot=<0 0 0 1>}
 deform_hair_root_bk_mid {parent=head, loc=<14.9413 -13.1833 0.2521>, rot=<-0.0051 -0.7157 0.6983 -0.011>}
 dyn_hair_base_bk_mid {parent=deform_hair_root_bk_mid, loc=<6.96 0 0>, rot=<0 0 0 1>}
 dyn_hair_01_bk_mid {parent=dyn_hair_base_bk_mid, loc=<0 0 0>, rot=<0 0 0 1>}
 deform_hair_root_bk_l {parent=head, loc=<15.5454 -11.2908 -7.0471>, rot=<0.0234 -0.8766 0.4726 -0.0872>}
 dyn_hair_base_bk_l {parent=deform_hair_root_bk_l, loc=<7.2579 0 0>, rot=<0 0 0 1>}
 dyn_hair_01_bk_l {parent=dyn_hair_base_bk_l, loc=<0 0 0>, rot=<0 0 0 1>}
 deform_hair_root_l {parent=head, loc=<15.6052 -5.5665 -10.09>, rot=<0.0059 -0.9999 0.0152 -0.0014>}
 dyn_hair_base_02_l {parent=deform_hair_root_l, loc=<5.0641 0 0>, rot=<0 0 0 1>}
 dyn_hair_02_l {parent=dyn_hair_base_02_l, loc=<0 0 0>, rot=<0 0 0 1>}
 dyn_hair_02_02_l {parent=dyn_hair_base_02_l, loc=<15.8105 1.9475 -0.4146>, rot=<0 0 0 1>}
 deform_hair_root_bk_r {parent=head, loc=<14.7914 -11.325 7.6922>, rot=<0.0015 -0.5311 0.8473 -0.0016>}
 dyn_hair_base_bk_r {parent=deform_hair_root_bk_r, loc=<7.1915 0 0.7029>, rot=<0 0 0 1>}
 dyn_hair_01_bk_r {parent=dyn_hair_base_bk_r, loc=<0 0 0>, rot=<0 0 0 1>}
 deform_hair_root_r {parent=head, loc=<15.0803 -4.9001 9.9838>, rot=<0.0097 -0.0781 0.9967 -0.0178>}
 dyn_hair_base_02_r {parent=deform_hair_root_r, loc=<7.6351 0 0.0931>, rot=<0 0 0 1>}
 dyn_hair_02_r {parent=dyn_hair_base_02_r, loc=<0 0 0>, rot=<0 0 0 1>}
 dyn_hair_02_02_r {parent=dyn_hair_base_02_r, loc=<12.5132 -1.9473 -0.464>, rot=<0 0 0 1>}
 attach_cape {parent=spine_05, loc=<9.017 -11.714 0>, rot=<0.0616 -0.7044 -0.0616 -0.7044>}
 attach_backpack {parent=spine_05, loc=<-0.123 -13.473 0>, rot=<-0.0093 -0.707 0.0093 -0.707>}
 clavicle_l {parent=spine_05, loc=<4.3575 -0.1466 -2.3223>, rot=<0.0465 -0.7289 -0.009 -0.6829>}
 upperarm_l {parent=clavicle_l, loc=<15.5111 0 0>, rot=<0.0226 -0.3234 -0.0088 -0.9459>}
 upperarm_twist_01_l {parent=upperarm_l, loc=<0.0041 0 0>, rot=<0 0 0 -1>}
 clavicle_r {parent=spine_05, loc=<4.3575 -0.1466 2.3223>, rot=<0.7289 0.0465 -0.6829 0.009>}
 upperarm_r {parent=clavicle_r, loc=<-15.5111 0 0>, rot=<0.0226 -0.3234 -0.0088 -0.9459>}
 upperarm_twist_01_r {parent=upperarm_r, loc=<-0.0041 0 0>, rot=<0 0 0 1>}
]
