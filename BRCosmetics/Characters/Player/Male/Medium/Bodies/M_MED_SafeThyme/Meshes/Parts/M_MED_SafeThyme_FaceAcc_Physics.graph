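# Physics asset for the SafeThyme medium male character's face accessory mesh, covering its dynamic hair bones.
# Capsules cover upper spine, clavicles, neck and head, linked by free angular constraints.
# Zero-size base bodies anchor front, back and side hair to the head, each carrying a capsule with limited swing and twist.
# Front pieces swing less than back ones, and the 01 side strands add a looser second segment.
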PhysicsAsset /BRCosmetics/Characters/Player/Male/Medium/Bodies/M_MED_SafeThyme/Meshes/Parts/M_MED_SafeThyme_FaceAcc_Physics
bodies {
 clavicle_l [{shapeType=Capsule, length=14.0535, radius=9.6146, center=<9.9632 0 -1.4525>, rotation=<rot 90 0 0>}]
 clavicle_r [{shapeType=Capsule, length=14.0535, radius=9.6146, center=<-6.3742 0 1.4525>, rotation=<rot 90 0 0>}]
 dyn_hair_01_02_l [{shapeType=Capsule, length=8.9025, radius=2.4825, center=<4.9831 0.0902 0.0815>, rotation=<rot 89.3594 -178.9855 -179.5641>}]
 dyn_hair_01_02_r [{shapeType=Capsule, length=7.5115, radius=1.7415, center=<4.2004 -0.322 0.7505>, rotation=<rot 82.5411 -179.5113 -173.0173>}]
 dyn_hair_01_bk_l [{shapeType=Capsule, length=0, radius=7, center=<9.747 0.0521 3.8338>, rotation=<rot 80.3804 0.667 3.9746>}]
 dyn_hair_01_bk_mid [{shapeType=Capsule, length=15, radius=7, center=<12 0 5>, rotation=<rot 90 0 0>}]
 dyn_hair_01_bk_r [{shapeType=Capsule, length=1.6195, radius=7, center=<9.5469 0 2.404>, rotation=<rot 90 0 0>}]
 dyn_hair_01_fr_l [{shapeType=Capsule, length=7, radius=5, center=<10 0 1.2692>, rotation=<rot 90 0 0>}]
 dyn_hair_01_fr_mid [{shapeType=Capsule, length=7, radius=5, center=<10 0 -1.5185>, rotation=<rot 86.0407 0 0>}]
 dyn_hair_01_fr_r [{shapeType=Capsule, length=7, radius=5, center=<10 0 0.54>, rotation=<rot 90 0 0>}]
 dyn_hair_01_l [{shapeType=Capsule, length=9.1035, radius=2.2814, center=<5.2259 0.0866 0.1031>, rotation=<rot 89.3594 -178.9855 -179.5641>}]
 dyn_hair_01_r [{shapeType=Capsule, length=7.5115, radius=1.7415, center=<4.25 0.0049 -0.1106>, rotation=<rot 84.9208 -175.7467 -175.2214>}]
 dyn_hair_02_l [{shapeType=Capsule, length=1.0289, radius=7, center=<13.4895 -3.2659 0.0661>, rotation=<rot 90 0 0>}]
 dyn_hair_02_r [{shapeType=Capsule, length=2.6455, radius=5.3834, center=<10.9342 4.7401 1.8221>, rotation=<rot 90 0 0>}]
 dyn_hair_base_01_l [{shapeType=Capsule, length=0, radius=0}]
 dyn_hair_base_01_r [{shapeType=Capsule, length=0, radius=0}]
 dyn_hair_base_02_l [{shapeType=Capsule, length=0, radius=0}]
 dyn_hair_base_02_r [{shapeType=Capsule, length=0, radius=0}]
 dyn_hair_base_bk_l [{shapeType=Capsule, length=0, radius=0}]
 dyn_hair_base_bk_mid [{shapeType=Capsule, length=0, radius=0}]
 dyn_hair_base_bk_r [{shapeType=Capsule, length=0, radius=0}]
 dyn_hair_base_fr_l [{shapeType=Capsule, length=0, radius=0}]
 dyn_hair_base_fr_mid [{shapeType=Capsule, length=0, radius=0}]
 dyn_hair_base_fr_r [{shapeType=Capsule, length=0, radius=0}]
 head [{shapeType=Capsule, length=20.1151, radius=7.7335, rotation=<rot 90 0 0>}]
 neck_01 [{shapeType=Capsule, length=6.4892, radius=6.5599, center=<3.6704 -1.2288 0>, rotation=<rot 90 0 0>}]
 spine_05 [{shapeType=Capsule, length=15, radius=13, center=<-2.7196 -1.2213 0>}]
}
constraints [
 {bone1Name=dyn_hair_01_fr_r, bone2Name=dyn_hair_base_fr_r, swing1LimitDegrees=20, swing2LimitDegrees=25, twistLimitDegrees=5}
 {
  bone1Name=dyn_hair_base_01_l
  bone2Name=head
  swing1LimitDegrees=45
  swing1Motion=Free
  swing2LimitDegrees=45
  swing2Motion=Free
  twistLimitDegrees=45
  twistMotion=Free
 }
 {bone1Name=dyn_hair_01_l, bone2Name=dyn_hair_base_01_l, swing1LimitDegrees=20, swing2LimitDegrees=20, twistLimitDegrees=5}
 {
  bone1Name=dyn_hair_base_01_r
  bone2Name=head
  swing1LimitDegrees=45
  swing1Motion=Free
  swing2LimitDegrees=45
  swing2Motion=Free
  twistLimitDegrees=45
  twistMotion=Free
 }
 {bone1Name=dyn_hair_01_r, bone2Name=dyn_hair_base_01_r, swing1LimitDegrees=20, swing2LimitDegrees=20, twistLimitDegrees=5}
 {
  bone1Name=dyn_hair_base_02_l
  bone2Name=head
  swing1LimitDegrees=45
  swing1Motion=Free
  swing2LimitDegrees=45
  swing2Motion=Free
  twistLimitDegrees=45
  twistMotion=Free
 }
 {bone1Name=dyn_hair_02_l, bone2Name=dyn_hair_base_02_l, swing1LimitDegrees=45, swing2LimitDegrees=45, twistLimitDegrees=45}
 {
  bone1Name=dyn_hair_base_02_r
  bone2Name=head
  swing1LimitDegrees=45
  swing1Motion=Free
  swing2LimitDegrees=45
  swing2Motion=Free
  twistLimitDegrees=45
  twistMotion=Free
 }
 {bone1Name=dyn_hair_02_r, bone2Name=dyn_hair_base_02_r, swing1LimitDegrees=45, swing2LimitDegrees=45, twistLimitDegrees=45}
 {
  bone1Name=dyn_hair_base_bk_l
  bone2Name=head
  swing1LimitDegrees=45
  swing1Motion=Free
  swing2LimitDegrees=45
  swing2Motion=Free
  twistLimitDegrees=45
  twistMotion=Free
 }
 {bone1Name=dyn_hair_01_bk_l, bone2Name=dyn_hair_base_bk_l, swing1LimitDegrees=45, swing2LimitDegrees=45, twistLimitDegrees=10}
 {
  bone1Name=dyn_hair_base_bk_mid
  bone2Name=head
  swing1LimitDegrees=45
  swing1Motion=Free
  swing2LimitDegrees=45
  swing2Motion=Free
  twistLimitDegrees=45
  twistMotion=Free
 }
 {bone1Name=dyn_hair_01_bk_mid, bone2Name=dyn_hair_base_bk_mid, swing1LimitDegrees=45, swing2LimitDegrees=45, twistLimitDegrees=10}
 {
  bone1Name=dyn_hair_base_bk_r
  bone2Name=head
  swing1LimitDegrees=45
  swing1Motion=Free
  swing2LimitDegrees=45
  swing2Motion=Free
  twistLimitDegrees=45
  twistMotion=Free
 }
 {bone1Name=dyn_hair_01_bk_r, bone2Name=dyn_hair_base_bk_r, swing1LimitDegrees=45, swing2LimitDegrees=45, twistLimitDegrees=10}
 {
  bone1Name=dyn_hair_base_fr_l
  bone2Name=head
  swing1LimitDegrees=45
  swing1Motion=Free
  swing2LimitDegrees=45
  swing2Motion=Free
  twistLimitDegrees=45
  twistMotion=Free
 }
 {bone1Name=dyn_hair_01_fr_l, bone2Name=dyn_hair_base_fr_l, swing1LimitDegrees=20, swing2LimitDegrees=25, twistLimitDegrees=5}
 {
  bone1Name=dyn_hair_base_fr_mid
  bone2Name=head
  swing1LimitDegrees=45
  swing1Motion=Free
  swing2LimitDegrees=45
  swing2Motion=Free
  twistLimitDegrees=45
  twistMotion=Free
 }
 {bone1Name=dyn_hair_01_fr_mid, bone2Name=dyn_hair_base_fr_mid, swing1LimitDegrees=20, swing2LimitDegrees=30, twistLimitDegrees=5}
 {
  bone1Name=dyn_hair_base_fr_r
  bone2Name=head
  swing1LimitDegrees=45
  swing1Motion=Free
  swing2LimitDegrees=45
  swing2Motion=Free
  twistLimitDegrees=45
  twistMotion=Free
 }
 {
  bone1Name=head
  bone2Name=neck_01
  swing1LimitDegrees=45
  swing1Motion=Free
  swing2LimitDegrees=45
  swing2Motion=Free
  twistLimitDegrees=45
  twistMotion=Free
 }
 {
  bone1Name=clavicle_l
  bone2Name=spine_05
  swing1LimitDegrees=45
  swing1Motion=Free
  swing2LimitDegrees=45
  swing2Motion=Free
  twistLimitDegrees=45
  twistMotion=Free
 }
 {
  bone1Name=clavicle_r
  bone2Name=spine_05
  swing1LimitDegrees=45
  swing1Motion=Free
  swing2LimitDegrees=45
  swing2Motion=Free
  twistLimitDegrees=45
  twistMotion=Free
 }
 {bone1Name=dyn_hair_01_02_r, bone2Name=dyn_hair_01_r, swing1LimitDegrees=45, swing2LimitDegrees=45, twistLimitDegrees=10}
 {bone1Name=dyn_hair_01_02_l, bone2Name=dyn_hair_01_l, swing1LimitDegrees=45, swing2LimitDegrees=45, twistLimitDegrees=10}
 {
  bone1Name=clavicle_r
  bone2Name=clavicle_l
  swing1LimitDegrees=45
  swing1Motion=Free
  swing2LimitDegrees=45
  swing2Motion=Free
  twistLimitDegrees=45
  twistMotion=Free
 }
 {
  bone1Name=neck_01
  bone2Name=spine_05
  swing1LimitDegrees=45
  swing1Motion=Free
  swing2LimitDegrees=45
  swing2Motion=Free
  twistLimitDegrees=45
  twistMotion=Free
 }
]
